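; Actor component that links actor lifetimes on the server, destroying its owner or attached actors when what they attach to dies.
; On BeginPlay it starts a randomized delay, then polls World Partition streaming completion with limited retries, since no streaming-started event exists.
; Once streaming completes, it binds OnDied on each valid AttachToActors entry to destroy the owner, and on the owning BuildingActor to destroy ActorsToAttach.
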(event EventBeginPlay
  (Utilities|SwitchHasAuthority
    (:Authority
      (if (> (+ (Utilities|Array|Length (Variables|Default|GetAttachToActors)) (Utilities|Array|Length (Variables|Default|GetActorsToAttach))) 0)
        (Variables|Default|SetRandomCheckInterval (Math|Random|RandomFloatinRange 8.0 12.0)) ; Random so they don't run on the same frame
        (CallFunction|TryandSetTimer) ; We must wait because begin play happens before WP starts streaming so it returns as True for streaming being complete.  We have no event for isStreamingStarted so use a delay which sucks
      ))))

(event Custom|CheckForStreamingComplete
  (Utilities|Time|ClearandInvalidateTimerbyHandle (Variables|Default|GettimerHandle_CheckForComplete))
  (bind _returnvalue (Streaming|IsAllStreamingCompleted (WorldSubsystems|GetWorldPartitionSubsystem)))
  (if _returnvalue
    (CallFunction|SignUpforDeathEvents)
    (elif (< (Variables|Default|GetCheckCount) 5)
      (CallFunction|TryandSetTimer)
      (Math|Integer|IncrementInt (Variables|Default|GetCheckCount)) ; Only check a certain amount of times just in case
    )))

(event Custom|HandleAttacheeDeath
  (params (DamagedActor :type Actor%20Object%20Reference) (Damage :type Float%20%28single-precision%29) (InstigatedBy :type Controller%20Object%20Reference) (DamageCauser :type Actor%20Object%20Reference) (HitLocation :type Vector) (FHitComponent :type Primitive%20Component%20Object%20Reference) (BoneName :type Name) (Momentum :type Vector))
  (Actor|DestroyActor (Components|GetOwner)))

(event Custom|SignUpForDeathEvents
  (if (Networking|IsServer)
    (Utilities|FlowControl|Sequence
      (:then_0
        (for _array_element (Variables|Default|GetAttachToActors)
          (if (Utilities|IsValidSoftObjectReference _array_element)
            (Utilities|IsValid (Utilities|ResolveSoftReference _array_element)
              (:Is%20Valid
                (EventDispatchers|BindEventtoOnDied (Utilities|ResolveSoftReference _array_element) (delegate Custom|HandleAttacheeDeath)))))))
      (:then_1
        (if (Utilities|Array|IsValidIndex (Variables|Default|GetActorsToAttach))
          (bind _asbuilding_actor (Utilities|Casting|CastToBuildingActor (Components|GetOwner)))
          (EventDispatchers|BindEventtoOnDied _asbuilding_actor (delegate Custom|CustomEvent_1)))))))

(event Custom|CustomEvent_1
  (params (DamagedActor :type Actor%20Object%20Reference) (Damage :type Float%20%28single-precision%29) (InstigatedBy :type Controller%20Object%20Reference) (DamageCauser :type Actor%20Object%20Reference) (HitLocation :type Vector) (FHitComponent :type Primitive%20Component%20Object%20Reference) (BoneName :type Name) (Momentum :type Vector))
  (for _array_element (Variables|Default|GetActorsToAttach)
    (if (Utilities|IsValidSoftObjectReference _array_element)
      (Utilities|IsValid (Utilities|ResolveSoftReference _array_element)
        (:Is%20Valid
          (Actor|DestroyActor (Utilities|ResolveSoftReference _array_element)))))))

(fn TryAndSetTimer
  (if (not (Utilities|Time|IsValidTimerHandle (Variables|Default|GettimerHandle_CheckForComplete)))
    (bind _returnvalue (Utilities|Time|SetTimerbyEvent (EventDispatchers|CreateEvent :!event CheckForStreamingComplete) (Variables|Default|GetRandomCheckInterval) false true))
    (Variables|Default|SettimerHandle_CheckForComplete _returnvalue)))
